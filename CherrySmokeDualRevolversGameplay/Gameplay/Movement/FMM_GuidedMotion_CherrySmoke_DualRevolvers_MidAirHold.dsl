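; Guided motion movement mode logic for the CherrySmoke Dual Revolvers mid-air hold, applying a constant force to the owning pawn.
; Force direction combines the current horizontal acceleration with a tunable upward component, then normalizes it, while magnitude comes from a separate scalable float.
; On activation, it also writes that direction as the air dodge vector on the Dual Revolvers pawn component, if present.
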(event EventOnExtensionActivate
  (params (RuntimeContext :type MMERuntime%20Context%20Structure%20%28by%20ref%29))
  (|CallParentFunction RuntimeContext)
  (bind _ownerpawn (Variables|FortMovementMode_BaseExtLogic|GetOwnerPawn)
    (:then
      (Utilities|IsValid (Actor|GetComponentbyClass _ownerpawn "/CherrySmokeDualRevolversGameplay/Component/CherrySmoke_DualRevolvers_PawnComponent.CherrySmoke_DualRevolvers_PawnComponent_C")
        (:Is%20Valid
          (Variables|/CherrySmokeDualRevolversGameplay/Component/CherrySmoke_DualRevolvers_PawnComponent.CherrySmoke_DualRevolvers_PawnComponent_C|SetAirDodgeVector (CallFunction|GetDirectionInternal) (Actor|GetComponentbyClass _ownerpawn "/CherrySmokeDualRevolversGameplay/Component/CherrySmoke_DualRevolvers_PawnComponent.CherrySmoke_DualRevolvers_PawnComponent_C")))))
    (:else)))

(fn GetDirection_ConstantForce
  (params (GuidedMotionRuntimeData :type Fort%20Movement%20Mode%20Guided%20Motion%20Runtime%20Data%20Object%20Reference))
  (return (CallFunction|GetDirectionInternal)))

(fn GetDirection_Internal
  (bind (_returnvalue_x _returnvalue_y) (Math|Vector|BreakVector (Pawn|Components|CharacterMovement|GetCurrentAcceleration (Variables|FortMovementMode_BaseExtLogic|GetOwnerMovementComponent))))
  (return (Math|Vector|Normalize (Math|Vector|MakeVector _returnvalue_x _returnvalue_y (ScalableFloat|GetValueatLevel (Variables|Default|GetHF_UpForce))))))

(fn GetStrength_ConstantForce
  (params (GuidedMotionRuntimeData :type Fort%20Movement%20Mode%20Guided%20Motion%20Runtime%20Data%20Object%20Reference))
  (return (ScalableFloat|GetValueatLevel (Variables|Default|GetHF_Force))))
