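; Guided motion movement mode for the CherrySmoke Dual Revolvers neutral double jump, supplying direction and strength for a constant force.
; Both outputs derive from one movement vector, split into its normalized direction and length, with Z set by a scalable force value.
; XY keeps current velocity, unless the hangtime extension is active.
; During hangtime, XY follows horizontal velocity direction at one of two scalable magnitudes, chosen by a speed threshold.
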(fn GetDirection_ConstantForce
  (params (GuidedMotionRuntimeData :type Fort%20Movement%20Mode%20Guided%20Motion%20Runtime%20Data%20Object%20Reference))
  (return (Math|Vector|Normalize (CallFunction|GetMovementVector))))

(fn GetMovementVector
  (bind (_returnvalue_x _returnvalue_y) (Math|Vector|BreakVector (* (Math|Vector|Normalize2D(Vector) (Variables|/Script/Engine.MovementComponent|GetVelocity (Variables|FortMovementMode_BaseExtLogic|GetOwnerMovementComponent))) (Math|Float|SelectFloat (ScalableFloat|GetValueatLevel (Variables|Default|GetHF_HorizonalAddHighVelocity)) (ScalableFloat|GetValueatLevel (Variables|Default|GetHF_HorizonalAdd)) (> (Math|Vector|VectorLength (Math|Vector|MakeVector (.x (Variables|/Script/Engine.MovementComponent|GetVelocity (Variables|FortMovementMode_BaseExtLogic|GetOwnerMovementComponent))) (.y (Variables|/Script/Engine.MovementComponent|GetVelocity (Variables|FortMovementMode_BaseExtLogic|GetOwnerMovementComponent))))) (ScalableFloat|GetValueatLevel (Variables|Default|GetHF_VelocityCheck)))))))
  (bind _ownermovementcomponent (Variables|FortMovementMode_BaseExtLogic|GetOwnerMovementComponent))
  (if (Class|FortMovementCompCharacter|IsExtensionPushedwithTag _ownermovementcomponent (LocalVariables|GetHangtimeMMETag))
    ;[0.07407 0.849 0 0.5] Hangtime XY velocity considerations
    (return (Math|Vector|MakeVector _returnvalue_x _returnvalue_y (ScalableFloat|GetValueatLevel (Variables|Default|GetHF_Force))))
    (else
      (return (Math|Vector|MakeVector (.x (Variables|/Script/Engine.MovementComponent|GetVelocity (Variables|FortMovementMode_BaseExtLogic|GetOwnerMovementComponent))) (.y (Variables|/Script/Engine.MovementComponent|GetVelocity (Variables|FortMovementMode_BaseExtLogic|GetOwnerMovementComponent))) (ScalableFloat|GetValueatLevel (Variables|Default|GetHF_Force)))))))

(fn GetStrength_ConstantForce
  (params (GuidedMotionRuntimeData :type Fort%20Movement%20Mode%20Guided%20Motion%20Runtime%20Data%20Object%20Reference))
  (return (Math|Vector|VectorLength (CallFunction|GetMovementVector))))
