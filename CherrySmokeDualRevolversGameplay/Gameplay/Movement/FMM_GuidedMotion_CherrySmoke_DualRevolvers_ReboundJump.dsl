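; Guided motion movement mode for the CherrySmoke Dual Revolvers rebound jump, pushing the owning pawn up and opposite to its air dodge direction.
; On activation it calls StartRebound on the pawn's DualRevolvers component.
; GetMovementVector reverses that component's flattened air dodge vector, scales it by a backwards force scalable float and sets Z from an upwards one.
; The constant-force direction and strength functions return that vector's normal and length.
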(event EventOnExtensionActivate
  (params (RuntimeContext :type MMERuntime%20Context%20Structure%20%28by%20ref%29))
  (|CallParentFunction RuntimeContext)
  (bind _ownerpawn (Variables|FortMovementMode_BaseExtLogic|GetOwnerPawn)
    (:then
      (Utilities|IsValid (Actor|GetComponentbyClass _ownerpawn "/CherrySmokeDualRevolversGameplay/Component/CherrySmoke_DualRevolvers_PawnComponent.CherrySmoke_DualRevolvers_PawnComponent_C")
        (:Is%20Valid
          (Class|CherrySmokeDualRevolversPawnComponent|StartRebound (Actor|GetComponentbyClass _ownerpawn "/CherrySmokeDualRevolversGameplay/Component/CherrySmoke_DualRevolvers_PawnComponent.CherrySmoke_DualRevolvers_PawnComponent_C")))))
    (:else)))

(fn GetDirection_ConstantForce
  (params (GuidedMotionRuntimeData :type Fort%20Movement%20Mode%20Guided%20Motion%20Runtime%20Data%20Object%20Reference))
  (return (Math|Vector|Normalize (CallFunction|GetMovementVector))))

(fn GetMovementVector
  (bind _ownerpawn (Variables|FortMovementMode_BaseExtLogic|GetOwnerPawn)
    (:then
      (Utilities|IsValid (Actor|GetComponentbyClass _ownerpawn "/CherrySmokeDualRevolversGameplay/Component/CherrySmoke_DualRevolvers_PawnComponent.CherrySmoke_DualRevolvers_PawnComponent_C")
        (:Is%20Valid
          (bind (_returnvalue_x _returnvalue_y) (Math|Vector|BreakVector (* (Math|Vector|Normalize2D(Vector) (Variables|/CherrySmokeDualRevolversGameplay/Component/CherrySmoke_DualRevolvers_PawnComponent.CherrySmoke_DualRevolvers_PawnComponent_C|GetAirDodgeVector (Actor|GetComponentbyClass _ownerpawn "/CherrySmokeDualRevolversGameplay/Component/CherrySmoke_DualRevolvers_PawnComponent.CherrySmoke_DualRevolvers_PawnComponent_C"))) (* (ScalableFloat|GetValueatLevel (Variables|Default|GetHF_BackwardsForce)) -1.0))))
          (return _returnvalue_x _returnvalue_y (ScalableFloat|GetValueatLevel (Variables|Default|GetHF_UpwardsForce))))
        (:Is%20Not%20Valid
          _)))
    (:else
      _))
  (return "0, 0, 0"))

(fn GetStrength_ConstantForce
  (params (GuidedMotionRuntimeData :type Fort%20Movement%20Mode%20Guided%20Motion%20Runtime%20Data%20Object%20Reference))
  (return (Math|Vector|VectorLength (CallFunction|GetMovementVector))))
